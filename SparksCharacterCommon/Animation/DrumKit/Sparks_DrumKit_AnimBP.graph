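# Animation Blueprint for the Sparks drum kit, animating its parts as a Fortnite player pawn plays and moves.
# It connects to drum event and state interfaces via the player's anim instance, driving hit montages and hi-hat and kick pedal state machines.
# Movement adds procedural part offsets, shakes and leaning, plus a seat height adjustment.
# Anim Dynamics swings cymbals and toms, with each cymbal's simulation blended out briefly when hit.
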AnimBlueprint /SparksCharacterCommon/Animation/DrumKit/Sparks_DrumKit_AnimBP
target_skeleton=/SparksCharacterCommon/Instrument/DrumKit/Generic/Meshes/Generic_DrumKit_Skeleton
graphs [
 AnimGraph {nodes [
   n1 = Additive Identity Pose()
   AnimGraphNode_StateMachine_1 = State Machine() {is=title AnimGraphNode_StateMachine}
   n2 = Get Is Traversing() -> (Is Traversing:bool)
   Blend Poses by bool(
    BlendPose_1=Apply Additive(
     Base=Use cached pose 'DefaultPose'()
     Additive=Blend Poses by bool(BlendPose_0=@AnimGraphNode_StateMachine_1, BlendPose_1=@n1, bActiveValue=@n2)
    )
    bActiveValue=Get Traversal_UseProceduralMovement() -> (Traversal_UseProceduralMovement:bool)
   )
   n3 = Get HiHatOffset() -> (HiHatOffset:double)
   n4 = "Transform (Modify) Bone - Bone: hihat_top"!(
    ComponentPose=Local To Component(LocalPose=Use cached pose 'Post HiHat Layer'())
    Translation_X:double="Clamp (Float)"(Value=float * float(A=@n3, B=2.5) {is=title K2Node_PromotableOperator}, Min=0, Max=100)
    Translation_Y:double=0
    Translation_Z:double=0
   ) -> (Pose) {translationMode=BMM_Additive, translationSpace=BCS_BoneSpace, alphaBoolBlend {blendOption=Linear}, boneToModify {boneName=hihat_top}}
   n5 = Blend Poses by bool(
    BlendPose_0=Component To Local(ComponentPose=@n4)
    BlendPose_1=Use cached pose 'Post HiHat Layer'()
    BlendTime_1=Get TraversingBlendOutTime() -> (TraversingBlendOutTime:double)
    bActiveValue=Get Is Traversing()
   )
   n6 = Use cached pose 'KickPedal Cached Pose'()
   n7 = NOT Boolean(A="Get (a copy)"(Array:Array of Booleans="Get Cymbal_Hit "() -> ("Cymbal_Hit ":Array of Booleans)) -> (Output:bool))
   n8 = Anim Dynamics - Bone: hihat_bottom_cymbal!(
    ComponentPose=Local To Component(LocalPose=Layered blend per bone(BasePose=@n5, BlendPoses_0=@n6) {layerSetup [{branchFilters [{boneName=kickdrum_pedal}, {boneName=kickdrum_beater}]}]})
    bAlphaBoolEnabled:bool=@n7
    ExternalForce="0,0,0"
   ) -> (Pose) {
    alphaInputType=Bool
    angularDampingOverride=0.999
    angularSpringConstant=1200
    bAngularSpring=true
    bOverrideAngularDamping=true
    bUseGravityOverride=true
    simulationSpace=BoneRelative
    alphaBoolBlend {blendInTime=0.25, blendOutTime=1}
    boundBone {boneName=hihat_bottom_cymbal}
    componentAppliedLinearAccClamp=<800 800 800>
    componentLinearAccScale=<0.3 0.3 0.3>
    componentLinearVelScale=<0.5 0.5 0.5>
    gravityOverride {z=-600}
    physicsBodyDefinitions [{
      boundBone {boneName=hihat_bottom_cymbal}
      boxExtents=<2 10 10>
      constraintSetup {angularLimitsMax=<10 4 2>, angularLimitsMin=<-10 -10 -2>, angularTarget=<1 0 0>}
      localJointOffset=<-1 0 0>
     }]
    relativeSpaceBone {boneName=hihat_stand}
    simSpaceSettings {maxAngularAcceleration=10, maxAngularVelocity=0.85, simSpaceAngularAlpha=0.25}
   }
   n9 = Anim Dynamics - Bone: hihat_top_cymbal!(ComponentPose=@n8, bAlphaBoolEnabled:bool=@n7, ExternalForce="0,0,0") -> (Pose) {
    alphaInputType=Bool
    angularDampingOverride=0.999
    angularSpringConstant=1200
    bAngularSpring=true
    bOverrideAngularDamping=true
    bUseGravityOverride=true
    simulationSpace=BoneRelative
    alphaBoolBlend {blendInTime=0.25, blendOutTime=1}
    boundBone {boneName=hihat_top_cymbal}
    componentAppliedLinearAccClamp=<800 800 800>
    componentLinearAccScale=<0.3 0.3 0.3>
    componentLinearVelScale=<0.5 0.5 0.5>
    gravityOverride {z=-600}
    physicsBodyDefinitions [{
      boundBone {boneName=hihat_top_cymbal}
      boxExtents=<2 10 10>
      constraintSetup {angularLimitsMax=<10 10 2>, angularLimitsMin=<-10 -4 -2>, angularTarget=<1 0 0>}
      localJointOffset=<-1 0 0>
     }]
    relativeSpaceBone {boneName=hihat_stand}
    simSpaceSettings {maxAngularAcceleration=10, maxAngularVelocity=0.85, simSpaceAngularAlpha=0.25}
   }
   n10 = Layered blend per bone(BasePose=Component To Local(ComponentPose=@n9), BlendPoses_0=Use cached pose 'KickPedal Cached Pose'()) {layerSetup [{branchFilters [{boneName=kickdrum_head}]}]}
   n11 = Slot 'Additive'(Source=Additive Identity Pose()) {slotName=Additive}
   n12 = "Transform (Modify) Bone - Bone: throne_seat"!(
    ComponentPose=Local To Component(LocalPose=Apply Additive(Base=@n10, Additive=@n11) {alphaBoolBlend {blendOption=Linear}})
    Translation_X:double=0
    Translation_Y:double=0
    Translation_Z:double=Get DrumSeatHeightZ() -> (DrumSeatHeightZ:double)
   ) -> (Pose) {translationMode=BMM_Additive, alphaBoolBlend {blendOption=Linear}, boneToModify {boneName=throne_seat}}
   Output Pose(Result=Blend Poses by bool(
     BlendPose_0=Component To Local(ComponentPose=@n12)
     BlendPose_1=Local Space Ref Pose()
     bActiveValue=Get SparksDrumsAnimBPValid() -> (SparksDrumsAnimBPValid:bool)
    ) {blendType=Linear}) {layerGroup=None}
   AnimGraphNode_StateMachine_2 = KickPedal() {is=title AnimGraphNode_StateMachine}
   Save cached pose 'KickPedal Cached Pose'(Pose=@AnimGraphNode_StateMachine_2) {cacheName=KickPedal Cached Pose}
   n13 = Get Leaning() -> (Leaning:double)
   n14 = Local To Component(LocalPose=Use cached pose 'DefaultPose'())
   n15 = Get Traversal_TranslatePart() -> (Traversal_TranslatePart:Array of Vectors)
   n16 = "Get (a copy)"(Array:Array of Vectors=@n15) -> (Output:Vector)
   n17 = "Transform (Modify) Bone - Bone: snare_stand"!(
    ComponentPose="Transform (Modify) Bone - Bone: crash_stand_l"!(
     ComponentPose="Transform (Modify) Bone - Bone: hihat_stand"!(ComponentPose=@n14, Translation=@n16) -> (Pose) {translationMode=BMM_Additive, alphaBoolBlend {blendOption=Linear}, boneToModify {boneName=hihat_stand}}
     Translation="Get (a copy)"(Array=@n15, Dimension 1=1)
    ) -> (Pose) {translationMode=BMM_Additive, alphaBoolBlend {blendOption=Linear}, boneToModify {boneName=crash_stand_l}}
    Translation="Get (a copy)"(Array=@n15, Dimension 1=2)
   ) -> (Pose) {translationMode=BMM_Additive, alphaBoolBlend {blendOption=Linear}, boneToModify {boneName=snare_stand}}
   n18 = "Get (a copy)"(Array=@n15, Dimension 1=3)
   n19 = "Transform (Modify) Bone - Bone: crash_stand_r"!(
    ComponentPose="Transform (Modify) Bone - Bone: floortom_stand"!(
     ComponentPose="Transform (Modify) Bone - Bone: kickdrum"!(ComponentPose=@n17, Translation=@n18) -> (Pose) {translationMode=BMM_Additive, alphaBoolBlend {blendOption=Linear}, boneToModify {boneName=kickdrum}}
     Translation="Get (a copy)"(Array=@n15, Dimension 1=4)
    ) -> (Pose) {translationMode=BMM_Additive, alphaBoolBlend {blendOption=Linear}, boneToModify {boneName=floortom_stand}}
    Translation="Get (a copy)"(Array=@n15, Dimension 1=5)
   ) -> (Pose) {translationMode=BMM_Additive, alphaBoolBlend {blendOption=Linear}, boneToModify {boneName=crash_stand_r}}
   n20 = "Get (a copy)"(Array=@n15, Dimension 1=6)
   n21 = Get Traversal_RotatePart() -> (Traversal_RotatePart:Array of Rotators)
   n22 = "Transform (Modify) Bone - Bone: ride_shake"!(
    ComponentPose="Transform (Modify) Bone - Bone: throne_stand"!(
     ComponentPose="Transform (Modify) Bone - Bone: ride_stand"!(ComponentPose=@n19, Translation=@n20) -> (Pose) {translationMode=BMM_Additive, alphaBoolBlend {blendOption=Linear}, boneToModify {boneName=ride_stand}}
     Translation="Get (a copy)"(Array=@n15, Dimension 1=7)
    ) -> (Pose) {translationMode=BMM_Additive, alphaBoolBlend {blendOption=Linear}, boneToModify {boneName=throne_stand}}
    Rotation="Get (a copy)"(Array:Array of Rotators=@n21, Dimension 1=6) -> (Output:Rotator)
   ) -> (Pose) {rotationMode=BMM_Additive, rotationSpace=BCS_BoneSpace, alphaBoolBlend {blendOption=Linear}, boneToModify {boneName=ride_shake}}
   n23 = "Transform (Modify) Bone - Bone: crash_shake_r"!(ComponentPose=@n22, Rotation="Get (a copy)"(Array=@n21, Dimension 1=5)) -> (Pose) {rotationMode=BMM_Additive, rotationSpace=BCS_BoneSpace, alphaBoolBlend {blendOption=Linear}, boneToModify {boneName=crash_shake_r}}
   n24 = "Transform (Modify) Bone - Bone: crash_shake_l"!(ComponentPose=@n23, Rotation="Get (a copy)"(Array=@n21, Dimension 1=1)) -> (Pose) {rotationMode=BMM_Additive, rotationSpace=BCS_BoneSpace, alphaBoolBlend {blendOption=Linear}, boneToModify {boneName=crash_shake_l}}
   n25 = float * float(A="Get (a copy)"!(Array=@n21, Dimension 1=3) -> (Output_Roll:float, Output_Pitch:float, Output_Yaw:float).Output_Pitch, B=1.7)
   n26 = float * float(A=@n25, B=-0.5)
   n27 = "Transform (Modify) Bone - Bone: tom_shake_r"!(ComponentPose=@n24, Rotation=MakeRotator(Roll=float * float(A=@n25, B=0.5), Pitch=@n26)) -> (Pose) {rotationMode=BMM_Additive, rotationSpace=BCS_BoneSpace, alphaBoolBlend {blendOption=Linear}, boneToModify {boneName=tom_shake_r}}
   n28 = "Transform (Modify) Bone - Bone: tom_shake_l"!(ComponentPose=@n27, Rotation=MakeRotator(Roll=@n26, Pitch=@n26)) -> (Pose) {rotationMode=BMM_Additive, rotationSpace=BCS_BoneSpace, alphaBoolBlend {blendOption=Linear}, boneToModify {boneName=tom_shake_l}}
   n29 = "Transform (Modify) Bone - Bone: snare_shake"!(
    ComponentPose=@n28
    Rotation="Scale Rotator (Float)"(A:Rotator="Get (a copy)"!(Array=@n21, Dimension 1=2) -> (Output:Rotator), B:float=-1) -> (ReturnValue:Rotator=~) {is=title K2Node_PromotableOperator}
   ) -> (Pose) {rotationMode=BMM_Additive, rotationSpace=BCS_BoneSpace, alphaBoolBlend {blendOption=Linear}, boneToModify {boneName=snare_shake}}
   n30 = Apply Additive(
    Base=Component To Local(ComponentPose=@n29)
    Additive=Blend Poses by bool(
     BlendPose_0=Blendspace Player 'Drumkit_Traversal_SideLean'!(X=@n13) -> (Pose) {blendSpace=&/Drumkit_Traversal_SideLean}
     BlendPose_1=Additive Identity Pose()
     BlendTime_1=Get LeanBlendOutTime() -> (LeanBlendOutTime:float)
     bActiveValue=Get Is Traversing()
    )
    Alpha=Get LeanAlpha() -> (LeanAlpha:float)
   )
   n31 = Use cached pose 'DefaultPose'()
   n32 = Get Is Traversing()
   n33 = Get TraversingBlendOutTime()
   n34 = "Get Cymbal_Hit "()
   n35 = Anim Dynamics - Bone: crash_cymbal_l!(
    ComponentPose=Local To Component(LocalPose=Blend Poses by bool(BlendPose_0=@n30, BlendPose_1=@n31, BlendTime_1=@n33, bActiveValue=@n32))
    bAlphaBoolEnabled:bool=NOT Boolean(A="Get (a copy)"(Array:Array of Booleans=@n34, Dimension 1=1) -> (Output:bool))
    ExternalForce="0,0,0"
   ) -> (Pose) {
    alphaInputType=Bool
    angularDampingOverride=0.9999
    angularSpringConstant=1200
    bAngularSpring=true
    bOverrideAngularDamping=true
    bUseGravityOverride=true
    simulationSpace=BoneRelative
    alphaBoolBlend {blendInTime=0.2, blendOutTime=0.75}
    boundBone {boneName=crash_cymbal_l}
    componentAppliedLinearAccClamp=<800 800 800>
    componentLinearAccScale=<0.3 0.3 0.3>
    componentLinearVelScale=<0.25 0.25 0.25>
    gravityOverride {z=-800}
    physicsBodyDefinitions [{
      boundBone {boneName=crash_cymbal_l}
      boxExtents=<11.25 11.25 4>
      constraintSetup {twistAxis=AxisZ, angularLimitsMax=<10 12 10>, angularLimitsMin=<-20 -16 -10>, angularTarget=<1 0 0>}
      localJointOffset=<0 0 -1>
     }]
    relativeSpaceBone {boneName=handle}
    simSpaceSettings {maxAngularAcceleration=10, maxAngularVelocity=0.85, simSpaceAngularAlpha=0.25}
   }
   n36 = "Get Cymbal_Hit "()
   n37 = Anim Dynamics - Bone: crash_cymbal_r!(ComponentPose=@n35, bAlphaBoolEnabled:bool=NOT Boolean(A="Get (a copy)"(Array=@n36, Dimension 1=2)), ExternalForce="0,0,0") -> (Pose) {
    alphaInputType=Bool
    angularDampingOverride=0.9999
    angularSpringConstant=1200
    bAngularSpring=true
    bOverrideAngularDamping=true
    bUseGravityOverride=true
    simulationSpace=BoneRelative
    alphaBoolBlend {blendInTime=0.2, blendOutTime=0.75}
    boundBone {boneName=crash_cymbal_r}
    componentAppliedLinearAccClamp=<800 800 800>
    componentLinearAccScale=<0.3 0.3 0.3>
    componentLinearVelScale=<0.25 0.25 0.25>
    gravityOverride {z=-800}
    physicsBodyDefinitions [{
      boundBone {boneName=crash_cymbal_r}
      boxExtents=<11.25 11.25 4>
      constraintSetup {twistAxis=AxisZ, angularLimitsMax=<10 12 10>, angularLimitsMin=<-20 -16 -10>, angularTarget=<1 0 0>}
      localJointOffset=<0 0 -1>
     }]
    relativeSpaceBone {boneName=handle}
    simSpaceSettings {maxAngularAcceleration=10, maxAngularVelocity=0.85, simSpaceAngularAlpha=0.25}
   }
   n38 = "Get Cymbal_Hit "()
   n39 = Anim Dynamics - Bone: ride_cymbal!(ComponentPose=@n37, bAlphaBoolEnabled:bool=NOT Boolean(A="Get (a copy)"(Array=@n38, Dimension 1=3)), ExternalForce="0,0,0") -> (Pose) {
    alphaInputType=Bool
    angularDampingOverride=0.99
    angularSpringConstant=1200
    bAngularSpring=true
    bOverrideAngularDamping=true
    bUseGravityOverride=true
    simulationSpace=BoneRelative
    alphaBoolBlend {blendInTime=0.25, blendOutTime=1.25}
    boundBone {boneName=ride_cymbal}
    componentAppliedLinearAccClamp=<600 600 600>
    componentLinearAccScale=<0.3 0.3 0.3>
    componentLinearVelScale=<0.2 0.25 0.2>
    gravityOverride {z=-600}
    physicsBodyDefinitions [{
      boundBone {boneName=ride_cymbal}
      boxExtents=<12 12 4>
      constraintSetup {twistAxis=AxisZ, angularLimitsMax=<10 8 10>, angularLimitsMin=<-20 -20 -10>, angularTarget=<1 0 0>}
      localJointOffset=<0 0 -1>
     }]
    relativeSpaceBone {boneName=handle}
    simSpaceSettings {maxAngularAcceleration=10, maxAngularVelocity=1, simSpaceAngularAlpha=0.25}
   }
   n40 = Anim Dynamics - Bone: tom_shake_r!(ComponentPose=@n39, bAlphaBoolEnabled:bool=True, ExternalForce="0,0,0") -> (Pose) {
    alphaInputType=Bool
    angularDampingOverride=0.9999
    angularSpringConstant=4000
    bAngularSpring=true
    bOverrideAngularDamping=true
    bUseGravityOverride=true
    simulationSpace=BoneRelative
    alphaBoolBlend {blendOutTime=0.3}
    boundBone {boneName=tom_shake_r}
    componentAppliedLinearAccClamp=<600 600 600>
    componentLinearAccScale=<1 1 1>
    componentLinearVelScale=<0.25 0.25 0.25>
    gravityOverride {z=-900}
    physicsBodyDefinitions [{
      boundBone {boneName=tom_shake_r}
      boxExtents=<12 12 12>
      constraintSetup {angularLimitsMax=<4 2 0>, angularLimitsMin=<-4 -2 0>, angularTarget=<1 0 0>}
      localJointOffset=<0 0 0.5>
     }]
    relativeSpaceBone {boneName=kickdrum}
    simSpaceSettings {maxAngularAcceleration=200, maxAngularVelocity=2, simSpaceAngularAlpha=1}
   }
   n41 = Anim Dynamics - Bone: tom_shake_l!(ComponentPose=@n40, bAlphaBoolEnabled:bool=True, ExternalForce="0,0,0") -> (Pose) {
    alphaInputType=Bool
    angularDampingOverride=0.9999
    angularSpringConstant=4000
    bAngularSpring=true
    bOverrideAngularDamping=true
    bUseGravityOverride=true
    simulationSpace=BoneRelative
    alphaBoolBlend {blendOutTime=0.3}
    boundBone {boneName=tom_shake_l}
    componentAppliedLinearAccClamp=<600 600 600>
    componentLinearAccScale=<1 1 1>
    componentLinearVelScale=<0.25 0.25 0.25>
    gravityOverride {z=-900}
    physicsBodyDefinitions [{
      boundBone {boneName=tom_shake_l}
      boxExtents=<12 12 12>
      constraintSetup {angularLimitsMax=<4 2 0>, angularLimitsMin=<-4 -2 0>, angularTarget=<1 0 0>}
      localJointOffset=<0 0 0.5>
     }]
    relativeSpaceBone {boneName=kickdrum}
    simSpaceSettings {maxAngularAcceleration=200, maxAngularVelocity=2, simSpaceAngularAlpha=1}
   }
   n42 = Anim Dynamics - Bone: floortom_shake!(ComponentPose=@n41, bAlphaBoolEnabled:bool=True, ExternalForce="0,0,0") -> (Pose) {
    alphaInputType=Bool
    angularDampingOverride=0.9999
    angularSpringConstant=500
    bAngularSpring=true
    bOverrideAngularDamping=true
    bUseGravityOverride=true
    simulationSpace=BoneRelative
    alphaBoolBlend {blendOutTime=0.2}
    boundBone {boneName=floortom_shake}
    componentAppliedLinearAccClamp=<600 600 600>
    componentLinearAccScale=<1 1 1>
    componentLinearVelScale=<0.4 0.4 0.4>
    gravityOverride {z=-900}
    physicsBodyDefinitions [{
      boundBone {boneName=floortom_shake}
      boxExtents=<40 40 40>
      constraintSetup {angularLimitsMax=<2 2 2>, angularLimitsMin=<-1 -1 -1>, angularTarget=<1 0 0>}
      localJointOffset=<1 0 0>
     }]
    relativeSpaceBone {boneName=floortom_stand}
    simSpaceSettings {maxAngularAcceleration=200, maxAngularVelocity=10, simSpaceAngularAlpha=1}
   }
   n43 = Component To Local(ComponentPose=@n42)
   n44 = Use cached pose 'DefaultPose'()
   AnimGraphNode_StateMachine_0 = HiHatPedal() {is=title AnimGraphNode_StateMachine}
   Save cached pose 'Post HiHat Layer'(Pose=Layered blend per bone(BasePose=Blend Poses by bool(BlendPose_0=@n43, BlendPose_1=@n44), BlendPoses_0=@AnimGraphNode_StateMachine_0) {layerSetup [{branchFilters [{boneName=hihat_top}, {boneName=hihat_pedal}]}]}) {cacheName=Post HiHat Layer}
   Save cached pose 'DefaultPose'(Pose=Slot 'DefaultSlot'(Source=Local Space Ref Pose())) {cacheName=DefaultPose}
  ]}
 HiHatPedal {
  entry=HiHat_ClosedHold
  path=AnimGraph.AnimGraphNode_StateMachine_0.HiHatPedal
  ids=id
  nodes [
   AnimStateNode_0 = HiHat_ClosedHold() {is=title AnimStateNode}
   AnimStateNode_1 = HitHat_Open() {is=title AnimStateNode}
   AnimStateNode_2 = HiHat_OpenHold() {is=title AnimStateNode}
   AnimStateNode_3 = HiHat_Close() {is=title AnimStateNode, stateEntered {guid=#1, notifyName=EnteredState_HiHatClose}}
   AnimStateTransitionNode_10 = AnimStateTransitionNode() {bAutomaticRuleBasedOnSequencePlayerInState=true, crossfadeDuration=0}
   AnimStateTransitionNode_11 = AnimStateTransitionNode() {crossfadeDuration=0.1}
   AnimStateTransitionNode_12 = AnimStateTransitionNode() {crossfadeDuration=0.1}
   AnimStateTransitionNode_7 = AnimStateTransitionNode() {crossfadeDuration=0}
   AnimStateTransitionNode_8 = AnimStateTransitionNode() {bAutomaticRuleBasedOnSequencePlayerInState=true, crossfadeDuration=0}
   AnimStateTransitionNode_9 = AnimStateTransitionNode() {crossfadeDuration=0}
  ]
  links [
   {
    dst=HiHat_ClosedHold
    rule=AnimGraph.AnimGraphNode_StateMachine_0.HiHatPedal.AnimStateTransitionNode_10.Transition
    src=HiHat_Close
    via=AnimStateTransitionNode_10
   }
   {
    dst=HitHat_Open
    rule=AnimGraph.AnimGraphNode_StateMachine_0.HiHatPedal.AnimStateTransitionNode_11.Transition
    src=HiHat_Close
    via=AnimStateTransitionNode_11
   }
   {
    dst=HitHat_Open
    rule=AnimGraph.AnimGraphNode_StateMachine_0.HiHatPedal.AnimStateTransitionNode_7.Transition
    src=HiHat_ClosedHold
    via=AnimStateTransitionNode_7
   }
   {
    dst=HiHat_Close
    rule=AnimGraph.AnimGraphNode_StateMachine_0.HiHatPedal.AnimStateTransitionNode_9.Transition
    src=HiHat_OpenHold
    via=AnimStateTransitionNode_9
   }
   {
    dst=HiHat_Close
    rule=AnimGraph.AnimGraphNode_StateMachine_0.HiHatPedal.AnimStateTransitionNode_12.Transition
    src=HitHat_Open
    via=AnimStateTransitionNode_12
   }
   {
    dst=HiHat_OpenHold
    rule=AnimGraph.AnimGraphNode_StateMachine_0.HiHatPedal.AnimStateTransitionNode_8.Transition
    src=HitHat_Open
    via=AnimStateTransitionNode_8
   }
  ]
 }
 HiHat_ClosedHold {
  path=AnimGraph.AnimGraphNode_StateMachine_0.HiHatPedal.AnimStateNode_0.HiHat_ClosedHold
  nodes [Output Animation Pose(Result=Sequence Player 'DrumKit_HiHat_ClosedHold'() {bLoopAnimation=false, sequence=&/DrumKit_HiHat_ClosedHold})]
 }
 HitHat_Open {
  path=AnimGraph.AnimGraphNode_StateMachine_0.HiHatPedal.AnimStateNode_1.HitHat_Open
  nodes [Output Animation Pose(Result=Sequence Player 'DrumKit_HiHat_Open'() {bLoopAnimation=false, sequence=&/DrumKit_HiHat_Open})]
 }
 HiHat_OpenHold {
  path=AnimGraph.AnimGraphNode_StateMachine_0.HiHatPedal.AnimStateNode_2.HiHat_OpenHold
  nodes [Output Animation Pose(Result=Sequence Player 'DrumKit_HiHat_OpenHold'() {bLoopAnimation=false, sequence=&/DrumKit_HiHat_OpenHold})]
 }
 HiHat_Close {
  path=AnimGraph.AnimGraphNode_StateMachine_0.HiHatPedal.AnimStateNode_3.HiHat_Close
  nodes [Output Animation Pose(Result=Sequence Player 'DrumKit_HiHat_Close'() {bLoopAnimation=false, sequence=&/DrumKit_HiHat_Close})]
 }
 Transition {path=AnimGraph.AnimGraphNode_StateMachine_0.HiHatPedal.AnimStateTransitionNode_10.Transition, nodes [Result()]}
 Transition {
  path=AnimGraph.AnimGraphNode_StateMachine_0.HiHatPedal.AnimStateTransitionNode_11.Transition
  nodes [Result(bCanEnterTransition=Get Trigger Hi Hat Open() -> (Trigger Hi Hat Open:bool))]
 }
 Transition {
  path=AnimGraph.AnimGraphNode_StateMachine_0.HiHatPedal.AnimStateTransitionNode_12.Transition
  nodes [Result(bCanEnterTransition=Get Trigger Hi Hat Close() -> (Trigger Hi Hat Close:bool))]
 }
 Transition {
  path=AnimGraph.AnimGraphNode_StateMachine_0.HiHatPedal.AnimStateTransitionNode_7.Transition
  nodes [Result(bCanEnterTransition=Get Trigger Hi Hat Open())]
 }
 Transition {path=AnimGraph.AnimGraphNode_StateMachine_0.HiHatPedal.AnimStateTransitionNode_8.Transition, nodes [Result()]}
 Transition {
  path=AnimGraph.AnimGraphNode_StateMachine_0.HiHatPedal.AnimStateTransitionNode_9.Transition
  nodes [Result(bCanEnterTransition=Get Trigger Hi Hat Close())]
 }
 State Machine {entry=State, path=AnimGraph.AnimGraphNode_StateMachine_1.State Machine, ids=id, nodes [AnimStateNode_0 = State() {is=title AnimStateNode}]}
 State {
  path=AnimGraph.AnimGraphNode_StateMachine_1.State Machine.AnimStateNode_0.State
  nodes [Output Animation Pose(Result="Sequence Player (additive) 'Drumkit_Traversal_Moving_02_Anim'"() {sequence=&/Drumkit_Traversal_Moving_02_Anim})]
 }
 KickPedal {
  entry=Bass_Hit_Ready
  path=AnimGraph.AnimGraphNode_StateMachine_2.KickPedal
  ids=id
  nodes [
   AnimStateNode_0 = Bass_Hit_Ready() {is=title AnimStateNode}
   AnimStateNode_1 = Bass_Hit_A() {is=title AnimStateNode, bAlwaysResetOnEntry=true}
   AnimStateNode_2 = Bass_Hit_B() {is=title AnimStateNode, bAlwaysResetOnEntry=true}
   AnimStateTransitionNode_20 = AnimStateTransitionNode() {crossfadeDuration=0.05}
   AnimStateTransitionNode_21 = AnimStateTransitionNode() {bAutomaticRuleBasedOnSequencePlayerInState=true, crossfadeDuration=0}
   AnimStateTransitionNode_22 = AnimStateTransitionNode() {bAutomaticRuleBasedOnSequencePlayerInState=true, crossfadeDuration=0}
   AnimStateTransitionNode_23 = AnimStateTransitionNode() {crossfadeDuration=0}
   AnimStateTransitionNode_24 = AnimStateTransitionNode() {crossfadeDuration=0.05}
  ]
  links [
   {
    dst=Bass_Hit_B
    rule=AnimGraph.AnimGraphNode_StateMachine_2.KickPedal.AnimStateTransitionNode_20.Transition
    src=Bass_Hit_A
    via=AnimStateTransitionNode_20
   }
   {
    dst=Bass_Hit_Ready
    rule=AnimGraph.AnimGraphNode_StateMachine_2.KickPedal.AnimStateTransitionNode_22.Transition
    src=Bass_Hit_A
    via=AnimStateTransitionNode_22
   }
   {
    dst=Bass_Hit_A
    rule=AnimGraph.AnimGraphNode_StateMachine_2.KickPedal.AnimStateTransitionNode_24.Transition
    src=Bass_Hit_B
    via=AnimStateTransitionNode_24
   }
   {
    dst=Bass_Hit_Ready
    rule=AnimGraph.AnimGraphNode_StateMachine_2.KickPedal.AnimStateTransitionNode_21.Transition
    src=Bass_Hit_B
    via=AnimStateTransitionNode_21
   }
   {
    dst=Bass_Hit_B
    rule=AnimGraph.AnimGraphNode_StateMachine_2.KickPedal.AnimStateTransitionNode_23.Transition
    src=Bass_Hit_Ready
    via=AnimStateTransitionNode_23
   }
  ]
 }
 Bass_Hit_Ready {
  path=AnimGraph.AnimGraphNode_StateMachine_2.KickPedal.AnimStateNode_0.Bass_Hit_Ready
  nodes [Output Animation Pose(Result="Sequence Evaluator (by time) 'DrumKit_BassDrum_Kick'"!(ExplicitTime:float=0) -> (Pose:Pose Link Structure) {is=title AnimGraphNode_SequenceEvaluator, sequence=&/DrumKit_BassDrum_Kick})]
 }
 Bass_Hit_A {
  path=AnimGraph.AnimGraphNode_StateMachine_2.KickPedal.AnimStateNode_1.Bass_Hit_A
  nodes [Output Animation Pose(Result=Sequence Player 'DrumKit_BassDrum_Kick'!(StartPosition:float=Get Right Leg Max Delay Shortened By() -> (Right Leg Max Delay Shortened By:float)) -> (Pose) {bLoopAnimation=false, sequence=&DrumKit_BassDrum_Kick}) {
    node {stateEntryFunction {functionName=OnKickPedalEntry}}
    stateEntryFunction {bSelfContext=true, memberGuid=#2, memberName=OnKickPedalEntry}
   }]
 }
 Bass_Hit_B {
  path=AnimGraph.AnimGraphNode_StateMachine_2.KickPedal.AnimStateNode_2.Bass_Hit_B
  nodes [Output Animation Pose(Result=Sequence Player 'DrumKit_BassDrum_Kick'(StartPosition=Get Right Leg Max Delay Shortened By()) {bLoopAnimation=false, sequence=&DrumKit_BassDrum_Kick}) {
    node {stateEntryFunction {functionName=OnStateEntry}}
    stateEntryFunction {bSelfContext=true, memberGuid=#2, memberName=OnKickPedalEntry}
   }]
 }
 Transition {
  path=AnimGraph.AnimGraphNode_StateMachine_2.KickPedal.AnimStateTransitionNode_20.Transition
  nodes [Result(bCanEnterTransition=ShouldKickPedalTransition!(self) -> (Transition:bool))]
 }
 Transition {path=AnimGraph.AnimGraphNode_StateMachine_2.KickPedal.AnimStateTransitionNode_21.Transition, nodes [Result()]}
 Transition {path=AnimGraph.AnimGraphNode_StateMachine_2.KickPedal.AnimStateTransitionNode_22.Transition, nodes [Result()]}
 Transition {
  path=AnimGraph.AnimGraphNode_StateMachine_2.KickPedal.AnimStateTransitionNode_23.Transition
  nodes [Result(bCanEnterTransition=ShouldKickPedalTransition())]
 }
 Transition {
  path=AnimGraph.AnimGraphNode_StateMachine_2.KickPedal.AnimStateTransitionNode_24.Transition
  nodes [Result(bCanEnterTransition=ShouldKickPedalTransition())]
 }
 EventGraph {nodes [
   n1 = PrintString!(execute, InString=Cast to BandPlay Failed, bPrintToScreen, bPrintToLog, TextColor=(R=1,G=0.9609,B=0,A=1), Duration=20, Key) -> (then)
   Do Once(execute=@n11.then_0, Reset=[@n26.then, @n1.then, @n9.else])
   n2 = GetOwningActor()
   > n3 = Is Valid(InputObject=@n2)
   PrintString(execute=@n3.Is Not Valid, InString=Could not find owning actor for DrumKit_AnimBP, TextColor=(R=1,G=0.9609,B=0,A=1), Duration=20)
   n4 = Cast To FortPlayerPawn(execute=@n3.Is Valid, Object=@n2)
   > Set Fort Player Pawn(Fort Player Pawn:Fort Player Pawn Object Reference=@n4) -> (Output_Get:Fort Player Pawn Object Reference)
   n5 = Get Fort Player Pawn() -> (Fort Player Pawn:Fort Player Pawn Object Reference)
   > Set Skeletal Mesh Asset(Skeletal Mesh Asset:Skeletal Mesh Object Reference=GetSkeletalMeshAsset!(self:Skeletal Mesh Component Object Reference=Get Mesh(self=@n5)) -> (ReturnValue:Skeletal Mesh Object Reference)) -> (Output_Get:Skeletal Mesh Object Reference)
   > SetDrumSeatHeight()
   n6 = GetAnimInstance(self=Get Mesh(self=Get Fort Player Pawn()))
   > n7 = Branch(Condition=Does Object Implement Interface!(
     TestObject:Object Reference=@n6
     Interface:Interface Interface=/Game/Characters/Player/Male/Male_Avg_Base/AnimLayerInterface/PlayerAnimOverrideLayerInterface.PlayerAnimOverrideLayerInterface_C
    ) -> (ReturnValue:bool))
   > n8 = SetDrumsEventAndStateInterfaceByClass!(
    execute
    self
    AnimLayerClass:Fort Base Layer Anim Instance Soft Class Reference=Get BandPlayDrumsClass() -> (BandPlayDrumsClass:Fort Base Layer Anim Instance Soft Class Reference)
    PlayerAnimInstance:Anim Instance Object Reference=@n6
   ) -> (then, Success:bool)
   > n9 = Branch(Condition=@n8)
   > K2Node_Composite_0 = DrumKit Hit Events!(execute) -> (Out:Exec) {is=title K2Node_Composite}
   PrintString(
    execute=@n7.else
    InString=Player's anim instance does not implement PlayerAnimOverrideLayerInstance
    TextColor=(R=1,G=0.9609,B=0,A=1)
    Duration=20
   )
   PrintString(execute=@n4.CastFailed, InString=Cast to First FortPlayerPawn Failed, TextColor=(R=1,G=0.9609,B=0,A=1), Duration=20)
   n10 = Event BlueprintUpdateAnimation()
   > Set Traversal_LeapTime(Traversal_LeapTime:float=float + float(A=@n10.DeltaTimeX, B=Get Traversal_LeapTime() -> (Traversal_LeapTime:float))) -> (Output_Get:float)
   > n11 = Sequence!(execute) -> (then_0, then_1, then_2:Exec, then_3:Exec)
   Branch(
    execute=@n11.then_1
    Condition=IsValidInterface!(Interface:"Interface Interface (by ref)"=Get SparksDrumStateInterface() -> (SparksDrumStateInterface:Sparks Drum Anim BP State Interface Interface)) -> (ReturnValue:bool)
   )
   > n12 = Get_TriggerHiHatOpen!(execute, self:Sparks Drum Anim BP State Interface Interface=Get SparksDrumStateInterface()) -> (then, TriggerHiHatOpen:bool)
   > Set Trigger Hi Hat Open(Trigger Hi Hat Open:bool=@n12) -> (Output_Get:bool)
   > n13 = Get_TriggerHiHatClose!(execute, self:Sparks Drum Anim BP State Interface Interface=Get SparksDrumStateInterface()) -> (then, TriggerHiHatClose:bool)
   > Set Trigger Hi Hat Close(Trigger Hi Hat Close:bool=@n13) -> (Output_Get:bool)
   > n14 = Get_RightLegMaxDelayShortenedBy!(execute, self:Sparks Drum Anim BP State Interface Interface=Get SparksDrumStateInterface()) -> (then, RightLegMaxDelayShortenedBy:double)
   > Set Right Leg Max Delay Shortened By(Right Leg Max Delay Shortened By:float=@n14) -> (Output_Get:float)
   > n15 = Get_HitBassIndex!(execute, self:Sparks Drum Anim BP State Interface Interface=Get SparksDrumStateInterface()) -> (then, Index:int=0)
   > Set HitBassIndex(HitBassIndex:int=@n15) -> (Output_Get:int=0)
   Branch(execute=@n11.then_2, Condition=Get AllowTraversal() -> (AllowTraversal:bool))
   n16 = Get CurrentPawnSquaredSpeedXY!(self:Fort Player Pawn Object Reference=Get Fort Player Pawn()) -> (CurrentPawnSquaredSpeedXY:float)
   > n17 = Set Is Traversing(Is Traversing:bool=AND Boolean(A="float >= float"(A=@n16, B=1), B=Get AllowTraversal())) -> (Output_Get:bool)
   > n18 = Sequence!(execute) -> (then_0, then_1)
   n19 = Get Actor Rotation!(self:Actor Object Reference=Get Fort Player Pawn()) -> (ReturnValue:Rotator) {is=title K2Node_CallFunction}
   > n20 = Set Traversal_RotatorBuffer(Traversal_RotatorBuffer:Rotator=RInterpTo!(
     Current:Rotator=Get Traversal_RotatorBuffer() -> (Traversal_RotatorBuffer:Rotator)
     Target:Rotator=@n19
     DeltaTime:float=@n10.DeltaTimeX
     InterpSpeed:float=Get LeapInterp Speed() -> (LeapInterp Speed:float)
    ) -> (ReturnValue:Rotator)) -> (Output_Get:Rotator)
   n21 = BreakRotator(InRot="Delta (Rotator)"!(A:Rotator=@n20, B:Rotator=@n19) -> (ReturnValue:Rotator))
   > Set Traversal_LookingLeanLerp(Traversal_LookingLeanLerp:float="Clamp (Float)"(Value=float * float(A=@n21.Yaw, B=-0.5), Min=-1, Max=1)) -> (Output_Get:float)
   n22 = float + float(
    A=Get!() -> (Traversal_NormalizedVelocity_X:double, Traversal_NormalizedVelocity_Y:double, Traversal_NormalizedVelocity_Z:double).Traversal_NormalizedVelocity_Y
    B=Get Traversal_LookingLeanLerp() -> (Traversal_LookingLeanLerp:float)
   )
   > Set Leaning(Leaning:double="Clamp (Float)"(Value=float * float(A=@n22, B=0.7), Min=-1, Max=1)) -> (Output_Get:double)
   > GetDrumPartsVectorList!(execute, self) -> (then, Array:Array of Vectors, Output_Get:Array of Rotators)
   n23 = Branch(execute=@n18.then_1, Condition=@n17)
   > Do Once(Reset=@n25.Completed)
   > n24 = Set GameTimeOffset(GameTimeOffset:double=GetGameTimeInSeconds()) -> (Output_Get:double)
   n25 = Do Once(execute=@n23.else, Reset=@n24.then, Start Closed=true)
   n26 = Event Reinitialize() {is=title K2Node_Event}
   Reroute Node()
   > n27 = For Loop(LastIndex=2)
   n28 = "Get (a copy)"(Array:Array of Transforms=Get BaseBoneTransforms() -> (BaseBoneTransforms:Array of Transforms), Dimension 1=@n27) -> (Output:Transform)
   n29 = GetOwningComponent()
   n30 = Select!(Option 0:Name=hihat_stand, Option 1:Name=kickdrum, Option 2:Name=throne_stand, Index:int=@n27) -> (ReturnValue:Name)
   > Set Array Elem!(
    execute
    TargetArray:Array of Transforms=Get BoneAnimOffsets() -> (BoneAnimOffsets:Array of Transforms)
    Index:int=@n27
    Item:"Transform (by ref)"=ComposeTransforms!(
     A:"Transform (by ref)"=GetSocketTransform(self=@n29, InSocketName=@n30, TransformSpace=RTS_ParentBoneSpace)
     B:"Transform (by ref)"=InvertTransform!(T:"Transform (by ref)"=@n28) -> (ReturnValue:Transform)
    ) -> (ReturnValue:Transform)
    bSizeToFit:bool=true
   ) -> (then)
   PrintString(InString=Cast to DrumsSM Failed, TextColor=(R=1,G=0.9609,B=0,A=1), Duration=20)
  ]}
 DrumKit Hit Events {
  path=EventGraph.K2Node_Composite_0.DrumKit Hit Events
  nodes [
   Montage_Play!(
    execute
    self:Anim Instance Object Reference
    MontageToPlay:Anim Montage Object Reference=&/DrumKit_Crash1_Choke_M
    InPlayRate:float=1
    ReturnValueType:EMontagePlayReturnType Enum=MontageLength
    InTimeToStartMontageAt:float=0
    bStopAllMontages:bool=false
   ) -> (then, ReturnValue:float)
   > Branch(Condition=Get Print Debug for Drum Events() -> (Print Debug for Drum Events:bool))
   Montage_Play(
    MontageToPlay=&/DrumKit_Cowbell_Shake_M
    InPlayRate=1
    ReturnValueType=MontageLength
    InTimeToStartMontageAt=0
    bStopAllMontages=false
   )
   > Branch(Condition=Get Print Debug for Drum Events())
   n1 = Snare_Shake() {is=title K2Node_CustomEvent}
   > Montage_Stop!(
    execute
    self:Anim Instance Object Reference
    InBlendOutTime:float=0
    Montage:Anim Montage Object Reference=&/DrumKit_Snare_Shake_M
   ) -> (then)
   > Montage_Play(
    MontageToPlay=&DrumKit_Snare_Shake_M
    InPlayRate=1
    ReturnValueType=MontageLength
    InTimeToStartMontageAt=0
    bStopAllMontages=false
   )
   > Branch(Condition=Get Print Debug for Drum Events())
   HiHat_PedalDown() {is=title K2Node_CustomEvent}
   > Montage_Play(
    MontageToPlay=&/DrumKit_HiHat_PedalDown_M
    InPlayRate=1
    ReturnValueType=MontageLength
    InTimeToStartMontageAt=0
    bStopAllMontages=false
   )
   > Branch(Condition=Get Print Debug for Drum Events())
   n2 = Ride_HardHit() {is=title K2Node_CustomEvent}
   > Set Array Elem(TargetArray:Array of Booleans="Get Cymbal_Hit "(), Index=3, Item:"Boolean (by ref)"=true, bSizeToFit=false)
   > Montage_Play(
    MontageToPlay=&/DrumKit_Ride_HardHit_M
    InPlayRate=1
    ReturnValueType=MontageLength
    InTimeToStartMontageAt=0
    bStopAllMontages=false
   )
   > RetriggerableDelay(Duration=0.5)
   > Set Array Elem(TargetArray="Get Cymbal_Hit "(), Index=3, bSizeToFit=false)
   > Branch(Condition=Get Print Debug for Drum Events())
   n3 = Tom1_Shake() {is=title K2Node_CustomEvent}
   > Montage_Play(
    MontageToPlay=&/DrumKit_Tom1_Shake_M
    InPlayRate=1
    ReturnValueType=MontageLength
    InTimeToStartMontageAt=0
    bStopAllMontages=false
   )
   > Branch(Condition=Get Print Debug for Drum Events())
   n4 = Tom2_Shake() {is=title K2Node_CustomEvent}
   > Montage_Play(
    MontageToPlay=&/DrumKit_Tom2_Shake_M
    InPlayRate=1
    ReturnValueType=MontageLength
    InTimeToStartMontageAt=0
    bStopAllMontages=false
   )
   > Branch(Condition=Get Print Debug for Drum Events())
   n5 = FloorTom_Shake() {is=title K2Node_CustomEvent}
   > Montage_Play(
    MontageToPlay=&/DrumKit_FloorTom_Shake_M
    InPlayRate=1
    ReturnValueType=MontageLength
    InTimeToStartMontageAt=0
    bStopAllMontages=false
   )
   > Branch(Condition=Get Print Debug for Drum Events())
   n6 = HiHat_PedalUp() {is=title K2Node_CustomEvent}
   > Montage_Play(
    MontageToPlay=&/DrumKit_HiHat_PedalUp_M
    InPlayRate=1
    ReturnValueType=MontageLength
    InTimeToStartMontageAt=0
    bStopAllMontages=false
   )
   > Branch(Condition=Get Print Debug for Drum Events())
   Ride_SoftHit() {is=title K2Node_CustomEvent}
   > Set Array Elem(TargetArray="Get Cymbal_Hit "(), Index=3, Item=true, bSizeToFit=false)
   > Montage_Play(
    MontageToPlay=&/DrumKit_Ride_SoftHit_M
    InPlayRate=1
    ReturnValueType=MontageLength
    InTimeToStartMontageAt=0
    bStopAllMontages=false
   )
   > RetriggerableDelay(Duration=0.5)
   > Set Array Elem(TargetArray="Get Cymbal_Hit "(), Index=3, bSizeToFit=false)
   > Branch(Condition=Get Print Debug for Drum Events())
   n7 = HiHat_Hit() {is=title K2Node_CustomEvent}
   > Set Array Elem(TargetArray="Get Cymbal_Hit "(), Index=0, Item=true, bSizeToFit=false)
   > n8 = Get_TriggerHiHatOpen(self=Get SparksDrumStateInterface())
   > n9 = Branch(Condition=@n8)
   > Montage_Stop(InBlendOutTime=0, Montage=&/DrumKit_HiHat_Hit_Open_M)
   > n10 = Montage_Play(
    MontageToPlay=&DrumKit_HiHat_Hit_Open_M
    InPlayRate=1
    ReturnValueType=MontageLength
    InTimeToStartMontageAt=0
    bStopAllMontages=false
   )
   RetriggerableDelay(execute=[@n10.then, @n11.then], Duration=0.5)
   > Set Array Elem(TargetArray="Get Cymbal_Hit "(), Index=0, bSizeToFit=false)
   > Branch(Condition=Get Print Debug for Drum Events())
   Montage_Stop(execute=@n9.else, InBlendOutTime=0, Montage=&/DrumKit_HiHat_Hit_Closed_M)
   > Montage_Stop(InBlendOutTime=0, Montage=&DrumKit_HiHat_Hit_Open_M)
   > n11 = Montage_Play(
    MontageToPlay=&DrumKit_HiHat_Hit_Closed_M
    InPlayRate=1
    ReturnValueType=MontageLength
    InTimeToStartMontageAt=0
    bStopAllMontages=false
   )
   n12 = Crash1_HardHit() {is=title K2Node_CustomEvent}
   > Set Array Elem(TargetArray="Get Cymbal_Hit "(), Index=1, Item=true, bSizeToFit=false)
   > Montage_Play(
    MontageToPlay=&/DrumKit_Crash1_HardHit_M
    InPlayRate=1
    ReturnValueType=MontageLength
    InTimeToStartMontageAt=0
    bStopAllMontages=false
   )
   > RetriggerableDelay(Duration=0.5)
   > Set Array Elem(TargetArray="Get Cymbal_Hit "(), Index=1, bSizeToFit=false)
   > Branch(Condition=Get Print Debug for Drum Events())
   n13 = Crash1_SoftHit() {is=title K2Node_CustomEvent}
   > Set Array Elem(TargetArray="Get Cymbal_Hit "(), Index=1, Item=true, bSizeToFit=false)
   > Montage_Play(
    MontageToPlay=&/DrumKit_Crash1_SoftHit_M
    InPlayRate=1
    ReturnValueType=MontageLength
    InTimeToStartMontageAt=0
    bStopAllMontages=false
   )
   > RetriggerableDelay(Duration=0.5)
   > Set Array Elem(TargetArray="Get Cymbal_Hit "(), Index=1, bSizeToFit=false)
   > Branch(Condition=Get Print Debug for Drum Events())
   n14 = Crash2_HardHit() {is=title K2Node_CustomEvent}
   > Set Array Elem(TargetArray="Get Cymbal_Hit "(), Index=2, Item=true, bSizeToFit=false)
   > Montage_Stop(InBlendOutTime=0, Montage=&/DrumKit_Crash2_HardHit_M)
   > Montage_Play(
    MontageToPlay=&DrumKit_Crash2_HardHit_M
    InPlayRate=1
    ReturnValueType=MontageLength
    InTimeToStartMontageAt=0
    bStopAllMontages=false
   )
   > RetriggerableDelay(Duration=0.5)
   > Set Array Elem(TargetArray="Get Cymbal_Hit "(), Index=2, bSizeToFit=false)
   > Branch(Condition=Get Print Debug for Drum Events())
   n15 = Crash2_SoftHit() {is=title K2Node_CustomEvent}
   > Set Array Elem(TargetArray="Get Cymbal_Hit "(), Index=2, Item=true, bSizeToFit=false)
   > Montage_Play(
    MontageToPlay=&/DrumKit_Crash2_SoftHit_M
    InPlayRate=1
    ReturnValueType=MontageLength
    InTimeToStartMontageAt=0
    bStopAllMontages=false
   )
   > RetriggerableDelay(Duration=0.5)
   > Set Array Elem(TargetArray="Get Cymbal_Hit "(), Index=2, bSizeToFit=false)
   > Branch(Condition=Get Print Debug for Drum Events())
   n16 = Crash2_Choke() {is=title K2Node_CustomEvent}
   > Set Array Elem(TargetArray="Get Cymbal_Hit "(), Index=2, Item=true, bSizeToFit=false)
   > Montage_Play(
    MontageToPlay=&/DrumKit_Crash2_Choke_M
    InPlayRate=1
    ReturnValueType=MontageLength
    InTimeToStartMontageAt=0
    bStopAllMontages=false
   )
   > RetriggerableDelay(Duration=0.5)
   > Set Array Elem(TargetArray="Get Cymbal_Hit "(), Index=2, bSizeToFit=false)
   > Branch(Condition=Get Print Debug for Drum Events())
   AnimNotify_EnteredState_HiHatClose() {is=title K2Node_Event}
   > Montage_Stop(InBlendOutTime=0, Montage=&DrumKit_HiHat_Hit_Open_M)
   Inputs()
   > n17 = Sequence()
   > Bind_KickHit!(
    execute
    self:Sparks Drum Events Interface Interface=Get SparksDrumEventsInterface() -> (SparksDrumEventsInterface:Sparks Drum Events Interface Interface)
    Callback:"Delegate (by ref)"=BassDrum_Kick() {is=title K2Node_CustomEvent}
   ) -> (then)
   > Bind_CowbellHit!(
    execute
    self:Sparks Drum Events Interface Interface=Get SparksDrumEventsInterface()
    Callback:"Delegate (by ref)"=Cowbell_Shake() {is=title K2Node_CustomEvent}
   ) -> (then)
   > Bind_Crash1_Choke!(
    execute
    self:Sparks Drum Events Interface Interface=Get SparksDrumEventsInterface()
    Callback:"Delegate (by ref)"=Crash1_Choke() {is=title K2Node_CustomEvent}
   ) -> (then)
   > Bind_Crash1_HardHit!(execute, self:Sparks Drum Events Interface Interface=Get SparksDrumEventsInterface(), Callback:"Delegate (by ref)"=@n12) -> (then)
   > Bind_Crash1_SoftHit!(execute, self:Sparks Drum Events Interface Interface=Get SparksDrumEventsInterface(), Callback:"Delegate (by ref)"=@n13) -> (then)
   > Bind_Crash2_Choke!(execute, self:Sparks Drum Events Interface Interface=Get SparksDrumEventsInterface(), Callback:"Delegate (by ref)"=@n16) -> (then)
   > Bind_Crash2_HardHit!(execute, self:Sparks Drum Events Interface Interface=Get SparksDrumEventsInterface(), Callback:"Delegate (by ref)"=@n14) -> (then)
   > Bind_Crash2_SoftHit!(execute, self:Sparks Drum Events Interface Interface=Get SparksDrumEventsInterface(), Callback:"Delegate (by ref)"=@n15) -> (then)
   > Bind_FloorTomHit!(execute, self:Sparks Drum Events Interface Interface=Get SparksDrumEventsInterface(), Callback:"Delegate (by ref)"=@n5) -> (then)
   > Bind_HiHatHit!(execute, self:Sparks Drum Events Interface Interface=Get SparksDrumEventsInterface(), Callback:"Delegate (by ref)"=@n7) -> (then)
   > Bind_HiHatPedalUp!(execute, self:Sparks Drum Events Interface Interface=Get SparksDrumEventsInterface(), Callback:"Delegate (by ref)"=@n6) -> (then)
   > Bind_RideHardHit!(execute, self:Sparks Drum Events Interface Interface=Get SparksDrumEventsInterface(), Callback:"Delegate (by ref)"=@n2) -> (then)
   > Bind_SnareHit!(execute, self:Sparks Drum Events Interface Interface=Get SparksDrumEventsInterface(), Callback:"Delegate (by ref)"=@n1) -> (then)
   > Bind_Tom1_Hit!(execute, self:Sparks Drum Events Interface Interface=Get SparksDrumEventsInterface(), Callback:"Delegate (by ref)"=@n3) -> (then)
   > Bind_Tom2_Hit!(execute, self:Sparks Drum Events Interface Interface=Get SparksDrumEventsInterface(), Callback:"Delegate (by ref)"=@n4) -> (then)
   Outputs!(Out:Exec=@n17.then_1)
   PrintString(InString=Crash 1 Soft Hit, TextColor=(R=1,G=0.9812,B=0,A=1))
   PrintString(InString=Crash 2 Hard Hit, TextColor=(R=1,G=0.9812,B=0,A=1))
   PrintString(InString=Crash 2 Soft Hit, TextColor=(R=1,G=0.9812,B=0,A=1))
   PrintString(InString=Crash 1 Choke, TextColor=(R=1,G=0.9812,B=0,A=1))
   PrintString(InString=Crash 2 Choke, TextColor=(R=1,G=0.9812,B=0,A=1))
   PrintString(InString=Ride Cymbal Hit, TextColor=(R=1,G=0.9812,B=0,A=1))
   PrintString(InString=Tom 1 Hit, TextColor=(R=1,G=0.9812,B=0,A=1))
   PrintString(InString=Tom 2 Hit, TextColor=(R=1,G=0.9812,B=0,A=1))
   PrintString(InString=Floor Tom Hit, TextColor=(R=1,G=0.9812,B=0,A=1))
   PrintString(InString=Percussion, TextColor=(R=1,G=0.9812,B=0,A=1))
   PrintString(InString=Hi Hat Pedal Up, TextColor=(R=1,G=0.9812,B=0,A=1))
   PrintString(InString=Ride Cymbal Hit, TextColor=(R=1,G=0.9812,B=0,A=1))
   PrintString(InString=Hi Hat Pedal Down, TextColor=(R=1,G=0.9812,B=0,A=1))
   PrintString(InString=Hi Hat Hit, TextColor=(R=1,G=0.9812,B=0,A=1))
   PrintString(InString=Snare Hit, TextColor=(R=1,G=0.9812,B=0,A=1))
   PrintString(InString=Crash 1 Hard Hit, TextColor=(R=1,G=0.9812,B=0,A=1))
  ]
 }
 GetDrumPartsVectorList {nodes [
   GetDrumPartsVectorList() {is=title K2Node_FunctionEntry}
   n1 = Get Fort Player Pawn()
   > Set Traversal_Direction(Traversal_Direction:double=CalculateDirection!(
     Velocity:"Vector (by ref)"=GetVelocity!(self:Actor Object Reference=@n1) -> (ReturnValue:Vector)
     BaseRotation:"Rotator (by ref)"=Get Actor Rotation(self=@n1)
    ) -> (ReturnValue:float)) -> (Output_Get:double)
   > n2 = For Loop(LastIndex=7)
   n3 = Sequence(execute=@n2.Completed)
   n4 = float + float(A=float - float(A=GetGameTimeInSeconds(), B=Get GameTimeOffset() -> (GameTimeOffset:double)), B=-0.2)
   n5 = Get Traversal_LeapBaseCurve() -> (Traversal_LeapBaseCurve:Curve Vector Object Reference)
   > Set HiHatOffset(HiHatOffset:double=BreakVector(InVec=GetVectorValue!(self:Curve Vector Object Reference=@n5, InTime:float=@n4) -> (ReturnValue:Vector)).Y) -> (Output_Get:double)
   n6 = Set Traversal_TranslatePart(execute=@n3.then_1, Traversal_TranslatePart:Array of Vectors=Get PartValues() -> (PartValues:Array of Vectors)) -> (Output_Get:Array of Vectors)
   > n7 = Set Traversal_RotatePart(Traversal_RotatePart:Array of Rotators=Get RotatePartValues() -> (RotatePartValues:Array of Rotators)) -> (Output_Get:Array of Rotators)
   > Return Node!(execute, Array:Array of Vectors=@n6, Output_Get:Array of Rotators=@n7)
   n8 = Get Traversal_TranslatePart()
   n9 = "To Float (Integer)"(InInt=@n2)
   n10 = Get Traversal_LeapOffsetCurve() -> (Traversal_LeapOffsetCurve:Curve Float Object Reference)
   n11 = GetGameTimeInSeconds()
   n12 = Get GameTimeOffset()
   n13 = GetVectorValue(
    self=Get Traversal_LeapBaseCurve()
    InTime=float + float(A=float - float(A=@n11, B=@n12), B=GetFloatValue!(self:Curve Float Object Reference=@n10, InTime:float=@n9) -> (ReturnValue:float))
   )
   n14 = BreakVector(InVec=@n13)
   n15 = Rotate Vector!(
    A_X:double=0
    A_Y:double=@n14.Y
    A_Z:double=@n14.Z
    B_Roll:float=0
    B_Pitch:float=0
    B_Yaw:float=Get Traversal_Direction() -> (Traversal_Direction:double)
   ) -> (ReturnValue:Vector)
   n16 = "To Float (Boolean)"(InBool=Get Is Traversing())
   n17 = Get CurveAnimIntensity() -> (CurveAnimIntensity:double)
   Add(
    execute=@n2.LoopBody
    TargetArray:Array of Vectors=Get PartValues()
    NewItem:"Vector (by ref)"=VInterpTo!(
     Current:Vector="Get (a copy)"(Array:Array of Vectors=@n8, Dimension 1=@n2) -> (Output:Vector)
     Target:Vector=vector * vector!(A=@n15, B:float=5, C:double=@n16, D:double=@n17) -> (ReturnValue=~)
     DeltaTime:float=GetDeltaSeconds!(self:Anim Instance Object Reference) -> (ReturnValue:float)
     InterpSpeed:float=Get LeapInterp Speed()
    ) -> (ReturnValue:Vector)
   ) {is=title K2Node_CallArrayFunction}
   > Add!(
    execute
    TargetArray:Array of Rotators=Get RotatePartValues()
    NewItem_Roll:float=0
    NewItem_Pitch:float=float * float!(A=BreakVector(InVec=@n13), B=-1, C:double=@n17) -> (ReturnValue)
    NewItem_Yaw:float=0
   ) -> (then, ReturnValue)
  ]}
 OnKickPedalEntry {nodes [
   OnKickPedalEntry!() -> (then, Context:"Anim Update Context Structure (by ref)", Node:"Anim Node Reference Structure (by ref)") {is=title K2Node_FunctionEntry}
   > Set HitBassPlayed(HitBassPlayed:int=Get HitBassIndex() -> (HitBassIndex:int=0)) -> (Output_Get:int=0)
   > Return Node!(execute)
  ]}
 SetDrumSeatHeight {nodes [
   SetDrumSeatHeight() {is=title K2Node_FunctionEntry}
   > n1 = Is Valid(InputObject=Get Skeletal Mesh Asset() -> (Skeletal Mesh Asset:Skeletal Mesh Object Reference))
   n2 = Get Skeletal Mesh Asset()
   n3 = Get Skeletal Mesh Asset()
   > n4 = Branch(Condition=OR Boolean(
     A="Equal (Object)"(A=@n3, B=/Game/Characters/Player/Male/Medium/Base/SK_M_MALE_Base_Skeleton)
     B="Equal (Object)"(A=@n2, B=/Game/Characters/Player/Male/Medium/Base/SK_M_MALE_Armature_Base_Skeleton)
    ))
   > n5 = Set DrumSeatHeightZ(DrumSeatHeightZ:double=2) -> (Output_Get:double)
   Reroute Node(InputPin=[@n5.then, @n9.then, @n13.then, @n17.then])
   n6 = Get Skeletal Mesh Asset()
   n7 = Get Skeletal Mesh Asset()
   n8 = Branch(
    execute=@n4.else
    Condition=OR Boolean(
     A="Equal (Object)"(A=@n7, B=/Game/Characters/Player/Female/Medium/Base/SK_M_Female_Base_Skeleton)
     B="Equal (Object)"(A=@n6, B=/Game/Characters/Player/Female/Medium/Base/SK_M_Female_Armature_Base_Skeleton)
    )
   )
   > n9 = Set DrumSeatHeightZ(DrumSeatHeightZ=6)
   n10 = Get Skeletal Mesh Asset()
   n11 = Get Skeletal Mesh Asset()
   n12 = Branch(
    execute=@n8.else
    Condition=OR Boolean(
     A="Equal (Object)"(A=@n11, B=/Game/Characters/Player/Male/Large/Base/SK_L_MALE_Base_Skeleton)
     B="Equal (Object)"(A=@n10, B=/Game/Characters/Player/Male/Large/Base/SK_L_MALE_Armature_Base_Skeleton)
    )
   )
   > n13 = Set DrumSeatHeightZ(DrumSeatHeightZ=-4)
   n14 = Get Skeletal Mesh Asset()
   n15 = Get Skeletal Mesh Asset()
   n16 = Branch(
    execute=@n12.else
    Condition=OR Boolean(
     A="Equal (Object)"(A=@n15, B=/Game/Characters/Player/Female/Large/Base/SK_L_Female_Base_Skeleton)
     B="Equal (Object)"(A=@n14, B=/Game/Characters/Player/Female/Large/Base/SK_L_Female_Armature_Base_Skeleton)
    )
   )
   > n17 = Set DrumSeatHeightZ(DrumSeatHeightZ=2)
   PrintString(
    execute=@n16.else
    InString="Missing option for this Skeletal Mesh! \r\nAdd additional types."
    Duration=GetWorldDeltaSeconds!() -> (ReturnValue:double)
   )
   PrintString(execute=@n1.Is Not Valid, InString="Invalid Skeletal Mesh from Player Pawn!", Duration=GetWorldDeltaSeconds())
   n18 = Get Skeletal Mesh Asset()
   n19 = Get DrumSeatHeightZ()
   PrintString(
    InString=Append!(A="Skeletal Mesh == \r\n", B=GetDisplayName(Object=@n18), C:String="DrumSeatHeightZ == \r\n", D:String="To String (Float)"(InDouble=@n19)) -> (ReturnValue)
    Duration=GetWorldDeltaSeconds()
   )
  ]}
 SetDrumsEventAndStateInterfaceByClass {nodes [
   n1 = SetDrumsEventAndStateInterfaceByClass!() -> (then, AnimLayerClass:Fort Base Layer Anim Instance Soft Class Reference, PlayerAnimInstance:Anim Instance Object Reference) {is=title K2Node_FunctionEntry}
   n2 = Resolve Soft Reference!(Input:Fort Base Layer Anim Instance Soft Class Reference=@n1) -> (Output:Fort Base Layer Anim Instance Class Reference)
   > n3 = Branch(Condition=IsValidClass!(Class:Object Class Reference=@n2) -> (ReturnValue:bool))
   > n4 = Cast To SparksBaseLayerAnimInstance!(
    execute
    Object=GetLinkedAnimLayerInstanceByClass!(self:Anim Instance Object Reference=@n1.PlayerAnimInstance, InClass:Anim Instance Class Reference=@n2, bCheckForChildClass:bool=true) -> (ReturnValue:Anim Instance Object Reference)
   ) -> (then, CastFailed, AsSparks Base Layer Anim Instance:Sparks Base Layer Anim Instance Object Reference) {is=title K2Node_DynamicCast}
   > Set BandPlayLayer(BandPlayLayer:Sparks Base Layer Anim Instance Object Reference=@n4) -> (Output_Get:Sparks Base Layer Anim Instance Object Reference)
   > n5 = Get!(execute:Exec) -> (then:Exec, else:Exec, BandPlayLayer:Sparks Base Layer Anim Instance Object Reference)
   n6 = Get BandPlayLayer() -> (BandPlayLayer:Sparks Base Layer Anim Instance Object Reference)
   > Set LinkedAnimGraphValue(LinkedAnimGraphValue:Anim Instance Object Reference=GetLinkedAnimGraphInstanceByTag!(
     self:Anim Instance Object Reference=@n6
     InTag:Name=Get InstrumentSlotTag!(self:Sparks Base Layer Anim Instance Object Reference=@n6) -> (InstrumentSlotTag:Name=None)
    ) -> (ReturnValue:Anim Instance Object Reference)) -> (Output_Get:Anim Instance Object Reference)
   > n7 = Branch(Condition=IsValid(Object=Get LinkedAnimGraphValue() -> (LinkedAnimGraphValue:Anim Instance Object Reference)))
   > n8 = Cast To SparksDrumEventsInterface!(execute, Object=Get LinkedAnimGraphValue()) -> (then, CastFailed, AsSparks Drum Events Interface:Sparks Drum Events Interface Interface) {is=title K2Node_DynamicCast}
   > Set SparksDrumEventsInterface(SparksDrumEventsInterface:Sparks Drum Events Interface Interface=@n8) -> (Output_Get:Sparks Drum Events Interface Interface)
   > n9 = Cast To Sparks_Drum_AnimBP_StateInterface!(execute, Object=Get LinkedAnimGraphValue()) -> (then, CastFailed, AsSparks Drum Anim BP State Interface:Sparks Drum Anim BP State Interface Interface) {is=title K2Node_DynamicCast}
   > Set SparksDrumStateInterface(SparksDrumStateInterface:Sparks Drum Anim BP State Interface Interface=@n9) -> (Output_Get:Sparks Drum Anim BP State Interface Interface)
   > Set SparksDrumsAnimBPValid(SparksDrumsAnimBPValid:bool=true) -> (Output_Get:bool)
   > Return Node!(execute, Success:bool=true)
   Return Node(execute=[@n9.CastFailed, @n8.CastFailed, @n7.else, @n5.else, @n4.CastFailed, @n3.else], Success=false)
   Reroute Node(InputPin:Wildcard) -> (OutputPin:Wildcard)
  ]}
 ShouldKickPedalTransition {nodes [
   ShouldKickPedalTransition() {is=title K2Node_FunctionEntry}
   > Return Node!(execute, Transition:bool="Not Equal (Integer)"(A=Get HitBassIndex() -> (HitBassIndex=0), B=Get HitBassPlayed() -> (HitBassPlayed:int=0)))
  ]}
]
variables [
 AllowTraversal {type=boolean}
 BandPlayDrumsClass {type=Class, default=/Script/SparksCharacterCommonRuntime.SparksBaseLayerAnimInstance}
 BaseBoneTransforms {
  type="array<Transform>"
  default [
   {location=<39.566 -49.3335 0.4296>, rotation=<rot 89.6027 -0.0027 39.9932>, scale=<1 1 1>}
   {location=<53.1694 14.3295 0>, rotation=<rot 4.4217 -165.9811 0>, scale=<1 1 1>}
   {location=<-19.9996 0.0002 0>, rotation=<rot 90 90 0>, scale=<1 1 1>}
  ]
 }
 BoneAnimOffsets {
  type="array<Transform>"
  default [
   {location=<0 0 0>, rotation=<rot 0 0 0>, scale=<1 1 1>}
   {location=<0 0 0>, rotation=<rot 0 0 0>, scale=<1 1 1>}
   {location=<0 0 0>, rotation=<rot 0 0 0>, scale=<1 1 1>}
  ]
 }
 CurveAnimIntensity {type=number, default=1}
 "Cymbal_Hit " {type="array<boolean>", default [false,false,false,false]}
 Drum {type=MeshComponent, default=null}
 DrumSeatHeightZ {type=number}
 Fort Player Pawn {type=FortPlayerPawn}
 GameTimeOffset {type=number}
 HiHatOffset {type=number}
 HitBassIndex {type=integer}
 HitBassPlayed {type=integer}
 Is Traversing {type=boolean}
 LeanAlpha {type=number, default=0.75}
 LeanBlendOutTime {type=number, default=0.35}
 Leaning {type=number}
 LeapInterp Speed {type=number, default=9}
 Print Debug for Drum Events {type=boolean}
 Right Leg Max Delay Shortened By {type=number}
 Skeletal Mesh Asset {type=SkeletalMesh, default=null}
 SparksDrumEventsInterface {default=None}
 SparksDrumsAnimBPValid {type=boolean}
 SparksDrumStateInterface {default=None}
 Traversal_Direction {type=number}
 Traversal_LeapBaseCurve {type=CurveVector, default=&/Drumkit_Traversal_Leap_Anim}
 Traversal_LeapOffsetCurve {type=CurveFloat, default=&/Drumkit_Traversal_LeapOffsets_Anim}
 Traversal_LeapTime {type=number}
 Traversal_LookingLeanLerp {type=number}
 Traversal_NormalizedVelocity {type=Vector}
 Traversal_RotatePart {type="array<Rotator>", default []}
 Traversal_RotatorBuffer {type=Rotator}
 Traversal_TranslatePart {type="array<Vector>"}
 Traversal_UseProceduralMovement {type=boolean, default=true}
 TraversingBlendOutTime {type=number, default=0.3}
 Trigger Hi Hat Close {type=boolean}
 Trigger Hi Hat Open {type=boolean}
]
