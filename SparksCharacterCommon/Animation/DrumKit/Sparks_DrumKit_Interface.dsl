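; Sparks_DrumKit_Interface is an interface asset for the drum kit, filed under the Sparks character animation content.
; It declares a single function, Reinitialize, with no inputs, outputs or logic shown.
; The name suggests a call to reset drum kit state, with the behavior left to implementers.
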(fn Reinitialize)
